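# Niagara 3D gas-fluid template system that simulates multicolored fire on a GPU grid sourced from particles on a user-supplied static mesh.
# The source emitter spawns dense, short-lived particles with panning noise driving density and a red-green-blue color curve.
# Shrinking particles write density, temperature and radius as fluid source attributes.
# A master grid emitter has volume and mesh renderers, with user-exposed resolution, world size and bounds drawing.
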NiagaraSystem /NiagaraFluids/Templates/Gas/3D/Systems/Grid3D_Gas_ColoredFire
dependencies=null
emitters {ParticleSourceEmitter {
  simTarget=GPUComputeSim
  emitterUpdateScript [
   &EmitterState()
   &SpawnRate(SpawnRate=50000, Use Spawn Probability=false)
  ]
  particleSpawnScript [
   &InitializeParticle(
    Lifetime Max=&Multiply_Float(A=0.0333, B=5)
    Lifetime Min:NiagaraFloat=&Multiply_Float(A=0.0333, B=2)
    Lifetime Mode=::NewEnumerator1
    Sprite Size Mode=::NewEnumerator3
    Uniform Sprite Size=5
    UsePositionOffset=false
    Write Material Random=true
   )
   /Niagara/Modules/Spawn/Location/StaticMeshLocation(
    MaskSpawnGroup=false
    OffsetAlongNormal=false
    OverrideRandomSeed=false
    SamplePosOffset=false
    Static Mesh:NiagaraDataInterfaceStaticMesh=$User.StaticMesh
   )
   /Niagara/Modules/Update/Utility/VectorNoiseValue(
    Noise Frequency=25
    Pan Noise=true
    Pan Noise Field:Vector3f=<0 0 7.5>
    Remap Output Value:ENiagara_CurlNoiseRemapType=::NewEnumerator1
    Remap Range Input:Vector2f=<0.2 0.5>
    Remap Range Output:Vector2f=<0 1>
   )
   /SetVariables(Particles.Density:NiagaraFloat=&Power(a=$Output.VectorNoiseValue.NoiseX, b=3)) {isSetParameters=true}
   &VectorNoiseValue(
    Noise Frequency=15
    Pan Noise=true
    Pan Noise Field:Vector3f=<0 0 2>
    Remap Output Value:ENiagara_CurlNoiseRemapType=::NewEnumerator1
    Remap Range Output:Vector2f=<0 1>
   ) {module=VectorNoiseValue001}
   /SetVariables(Particles.Color=&ColorFromCurve(
     ColorCurve=DataInterface() {propertyValues {
       AlphaCurve {keys [<at 0 1 RCIM_Cubic>]}
       BlueCurve {keys [<at 0 0 RCIM_Cubic>,<at 0.5 0>,<at 1 1 RCIM_Cubic>]}
       GreenCurve {keys [<at 0 0 RCIM_Cubic>,<at 0.5 1>,<at 1 0 RCIM_Cubic>]}
       RedCurve {keys [<at 0 1 RCIM_Cubic>,<at 0.5 0>,<at 1 0 RCIM_Cubic>]}
      }}
     CurveIndex=$Output.VectorNoiseValue001.NoiseZ
    )) {isSetParameters=true}
  ]
  particleUpdateScript [
   &ParticleState()
   &Drag(Use Linear Drag=true, Use Rotational Drag=false)
   &SolveForcesAndVelocity(Clamp Velocity=false, Limit Acceleration=false)
   &ScaleSpriteSize(Scale Factor=&Vector2DFromFloat(Value=&FloatFromCurve(FloatCurve=DataInterface() {propertyValues {Curve {keys [
          {interpMode=RCIM_Cubic, leaveTangentWeight=0.3691, tangentMode=RCTM_User, time=0, value=1}
          {arriveTangent=-2.8786, arriveTangentWeight=0.3336, interpMode=RCIM_Cubic, leaveTangent=-2.8786, tangentMode=RCTM_User, time=1, value=0}
         ]}}})))
   /NiagaraFluids/Modules/Particles/SetFluidSourceAttributes(
    Density:NiagaraFloat=&Multiply_Float(A=$Particles.Density, B=1)
    Radius:NiagaraFloat=&Multiply_Float(A=&MakeFloatFromVector2d(Vector 2D=$Particles.SpriteSize), B=0.5)
    Temperature:NiagaraFloat=&Multiply_Float(A=$Particles.Density, B=1)
   )
  ]
  rendererClasses [&NiagaraSpriteRendererProperties]
  renderers [{rendererClass=&NiagaraSpriteRendererProperties}]
 }}
summary {
 systemName=Grid3D_Gas_ColoredFire
 emitters [
  {emitterName=ParticleSourceEmitter, simTarget=GPUComputeSim, rendererClasses [&NiagaraSpriteRendererProperties]}
  {
   emitterName=Grid3D_Gas_Master_Emitter
   simTarget=GPUComputeSim
   rendererClasses [&NiagaraMeshRendererProperties, /Script/Niagara.NiagaraVolumeRendererProperties]
  }
 ]
 userVariables [
  User.DrawBounds {type=NiagaraBool, defaultValue=true}
  User.ResolutionMaxAxis {type=NiagaraInt32, defaultValue=256}
  User.StaticMesh {
   type=NiagaraDataInterfaceStaticMesh
   description=""
   defaultValue {
    struct=NiagaraExt_VariableValue_DataInterface
    value {
     dataInterface=&self:NiagaraDataInterfaceStaticMesh_0
     dataInterfaceClass=&NiagaraDataInterfaceStaticMesh
    }
   }
  }
  User.WorldSpaceSize {type=Vector3f, description="", defaultValue {struct=Vector3f, value=<200 200 400>}}
 ]
}
