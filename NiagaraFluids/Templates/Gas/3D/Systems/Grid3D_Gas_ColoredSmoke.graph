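# Niagara Fluids 3D gas template system for colored smoke, pairing a GPU particle source emitter with a Grid3D gas master emitter.
# A burst of seeded particles follows a local SpinPosition module, displaced by panning curl noise, with velocity computed afterward.
# Color cycles through a red-green-blue ramp using spin sweep plus emitter age.
# SetFluidSourceAttributes marks particles as zero-temperature density and velocity sources; the master emitter has mesh and volume renderers.
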NiagaraSystem /NiagaraFluids/Templates/Gas/3D/Systems/Grid3D_Gas_ColoredSmoke
dependencies=null
emitters {ParticleSourceEmitter {
  simTarget=GPUComputeSim
  emitterUpdateScript [
   &EmitterState()
   &SpawnBurst_Instantaneous(Spawn Count=1000, Use Spawn Probability=false)
  ]
  particleSpawnScript [
   &InitializeParticle(
    Color=<rgba 1 0 0 1>
    Lifetime=$Emitter.CurrentLoopDuration
    Sprite Size Mode=::NewEnumerator3
    Uniform Sprite Size=4
    UsePositionOffset=false
    Write Material Random=true
   )
   /SetVariables(Particles.SeedPerParticle:NiagaraInt32=&RandomRangeInt(Minimum=0, Maximum=99999)) {isSetParameters=true}
  ]
  particleUpdateScript [
   &ParticleState()
   &self:SpinPosition(
    Axis:Vector3f=<0 1 0>
    Length=50
    Offset:Vector3f=<0 0 335>
    Radius=100
    Radius Randomize Gain=0.2
    Seed Per Particle:NiagaraInt32=$Particles.SeedPerParticle
    SpinSpeed=2
   ) {module=SpinPosition}
   /SetVariables(Particles.RampParam:NiagaraFloat=&ModuloFloat(
     A=&Add_Float(A=$Output.SpinPosition.SweepParam, B=&Multiply_Float(A=$Emitter.Age, B=0.1))
     B=1
    )) {isSetParameters=true}
   /SetVariables(Particles.Color=&ColorFromCurve(
     ColorCurve=DataInterface() {propertyValues {
       AlphaCurve {keys [<at 0 1 RCIM_Cubic>,<at 1 1 RCIM_Cubic>]}
       BlueCurve {keys [<at 0 0 RCIM_Cubic>, <at 0.3333 0>, <at 0.6666 1>, {arriveTangent=-2.9994, interpMode=RCIM_Cubic, leaveTangent=-2.9994, time=1, value=0}]}
       GreenCurve {keys [{arriveTangent=3.0003, interpMode=RCIM_Cubic, leaveTangent=3.0003, time=0, value=0}, <at 0.3333 1>, <at 0.6666 0>, <at 1 0 RCIM_Cubic>]}
       RedCurve {keys [
         {arriveTangent=-3.0003, interpMode=RCIM_Cubic, leaveTangent=-3.0003, time=0, value=1}
         <at 0.3333 0>
         <at 0.6666 0>
         {arriveTangent=2.9994, interpMode=RCIM_Cubic, leaveTangent=2.9994, time=1, value=1}
        ]}
      }}
     CurveIndex=$Particles.RampParam
    )) {isSetParameters=true}
   &ScaleColor(Color Value To Scale=$Particles.Color, Scale RGB=<2 2 2>, ScaleA=true, ScaleRGB=true)
   &CurlNoiseForce(
    Noise Frequency=10
    Noise Strength=1
    Pan Noise=true
    Pan Noise Field=<0 1 0>
    Sample Position:NiagaraPosition=&Multiply_Vector(A=Linked() {linkedVariable {name=Particles.Position, type=NiagaraPosition}}, B=<0 0 1>)
   )
   /SetVariables(Particles.Position=&Add_Vector(
     A=Linked() {linkedVariable {name=Particles.Position, type=NiagaraPosition}}
     B=&Multiply_Vector(A=$Output.CurlNoiseForce.CurlNoiseForce, B=<20 20 20>)
    )) {isSetParameters=true}
   &CalculateAccurateVelocity()
   /NiagaraFluids/Modules/Particles/SetFluidSourceAttributes(
    Density=0.4
    Radius:NiagaraFloat=&MakeFloatFromVector2d(Vector 2D=$Particles.SpriteSize)
    Temperature=0
    Velocity Scale=0.9
   )
  ]
  rendererClasses [&NiagaraSpriteRendererProperties]
  renderers [{rendererClass=&NiagaraSpriteRendererProperties}]
 }}
summary {
 systemName=Grid3D_Gas_ColoredSmoke
 emitters [
  {emitterName=ParticleSourceEmitter, simTarget=GPUComputeSim, rendererClasses [&NiagaraSpriteRendererProperties]}
  {
   emitterName=Grid3D_Gas_Master_Emitter
   simTarget=GPUComputeSim
   rendererClasses [&NiagaraMeshRendererProperties, /Script/Niagara.NiagaraVolumeRendererProperties]
  }
 ]
 userVariables [
  User.DrawBounds {type=NiagaraBool, defaultValue=true}
  User.ResolutionMaxAxis {type=NiagaraInt32, defaultValue=192}
  User.WorldSpaceSize {type=Vector3f, description="", defaultValue {struct=Vector3f, value=<450 225 450>}}
 ]
}
