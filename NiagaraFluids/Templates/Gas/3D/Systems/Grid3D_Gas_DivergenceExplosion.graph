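# Niagara Fluids 3D gas template system that simulates an explosion driven by divergence.
# A GPU source emitter bursts short-lived particles in a noise-distorted sphere, pushes them outward from the simulation position, then slows them with drag.
# Particles write fluid source density, temperature, radius and divergence, with noise-based density scaling a divergence pulse that rises then falls.
# Grid3D_Gas_Master_Emitter, summarized only, accompanies the source with mesh and volume renderers.
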NiagaraSystem /NiagaraFluids/Templates/Gas/3D/Systems/Grid3D_Gas_DivergenceExplosion
dependencies=null
emitters {ParticleSourceEmitter {
  simTarget=GPUComputeSim
  emitterUpdateScript [
   &EmitterState()
   &SpawnBurst_Instantaneous(
    Spawn Count=15
    Spawn Time=/Niagara/DynamicInputs/Math/Min_Float(A:NiagaraFloat=0.2, B:NiagaraFloat=$Engine.DeltaTime)
    Use Spawn Probability=false
    UseLoopCountLimit=false
   )
   /SetVariables(Emitter.SphereRadius=10) {isSetParameters=true}
   /SetVariables() {isSetParameters=true}
  ]
  particleSpawnScript [
   &InitializeParticle(
    Color=<rgba 1 0 0 1>
    Lifetime Max=&Multiply_Float(A=0.0333, B=20)
    Lifetime Min:NiagaraFloat=&Multiply_Float(A=0.0333, B=10)
    Lifetime Mode=::NewEnumerator1
    Sprite Size Mode=::NewEnumerator1
    UsePositionOffset=false
    Write Material Random=true
   )
   &ShapeLocation(MaskSpawnGroup=false, Override Seed=false, Sphere Radius=$Emitter.SphereRadius, UseOwnerScale=true)
   /Niagara/Modules/Update/Utility/VectorNoiseValue(Noise Frequency=100, Pan Noise=true, Pan Noise Field:Vector3f=<0 0 1>)
   &VectorNoiseValue(Pan Noise=true, Pan Noise Field:Vector3f=<0 0 5>) {module=VectorNoiseValue001}
   &self:ParticleSourceEmitter_0.NiagaraScratchPadContainer_2.OffsetPositionWithNoise_0(
    NoiseGain=0.5
    NoiseOffset=1
    NoiseValue:NiagaraFloat=&Add_Float(
     A=&Multiply_Float(A=$Output.VectorNoiseValue.NoiseX, B=1)
     B=&Multiply_Float(A=$Output.VectorNoiseValue001.NoiseX, B=1)
    )
   ) {module=OffsetPositionWithNoise}
   &AddVelocity(
    Velocity=&Subtract_Position(A=$Particles.Position, B=/Niagara/DynamicInputs/Helpers/SimulationPosition())
    Velocity Speed Scale=$Emitter.VelocityMultiplier
   )
   /SetVariables(Particles.Position=&AddVectorToPosition(Position=$Particles.Position, Vector=$User.SourceOffset)) {isSetParameters=true}
   &VectorNoiseValue(
    Noise Frequency=100
    Pan Noise=true
    Pan Noise Field:Vector3f=<0 0 1>
    Remap Output Value:ENiagara_CurlNoiseRemapType=::NewEnumerator1
    Remap Range Output:Vector2f=<0.2 1>
   ) {module=VectorNoiseValue002}
   /SetVariables(Particles.Density:NiagaraFloat=&Power(a=$Output.VectorNoiseValue002.NoiseX, b=3)) {isSetParameters=true}
   /SetVariables(Particles.Color=&MultiplyLinearColor(
     Linear Color=<rgba 1 1 1 1>
     Scale Factor=&MakeFloat4FromVectorAndFloat("Vector (RGB)"=&VectorFromFloat(Value=$Particles.Density), "Float (Alpha)"=1)
    )) {isSetParameters=true}
  ]
  particleUpdateScript [
   &ParticleState()
   &Drag(Use Linear Drag=true, Use Rotational Drag=false)
   &SolveForcesAndVelocity(Clamp Velocity=false, Limit Acceleration=false)
   &ScaleSpriteSize(Scale Factor=&Vector2DFromFloat(Value=&FloatFromCurve(FloatCurve=DataInterface() {propertyValues {Curve {keys [
          {interpMode=RCIM_Cubic, leaveTangentWeight=0.3691, tangentMode=RCTM_User, time=0, value=1}
          {arriveTangent=-2.8786, arriveTangentWeight=0.3336, interpMode=RCIM_Cubic, leaveTangent=-2.8786, tangentMode=RCTM_User, time=1, value=0}
         ]}}})))
   &KillParticles(Kill Particles=&SetBoolByFloatComparison(A=&VectorLength(Vector=$Particles.Velocity), B=1, Comparison Type=::NewEnumerator6))
   /NiagaraFluids/Modules/Particles/SetFluidSourceAttributes(
    Density:NiagaraFloat=$Particles.Density
    Divergence:NiagaraFloat=&Multiply_Float(
     A=$Particles.Density
     B=&FloatFromCurve(FloatCurve=DataInterface() {propertyValues {Curve {keys [<at 0 0>,<at 0.5 1>,<at 1 0>]}}}, Scale Curve=100)
    )
    Radius:NiagaraFloat=&Multiply_Float(A=&MakeFloatFromVector2d(Vector 2D=$Particles.SpriteSize), B=0.5)
    Temperature:NiagaraFloat=&Multiply_Float(A=$Particles.Density, B=1)
   )
  ]
  rendererClasses [&NiagaraSpriteRendererProperties]
  renderers [{rendererClass=&NiagaraSpriteRendererProperties}]
 }}
summary {
 systemName=Grid3D_Gas_DivergenceExplosion
 emitters [
  {emitterName=ParticleSourceEmitter, simTarget=GPUComputeSim, rendererClasses [&NiagaraSpriteRendererProperties]}
  {
   emitterName=Grid3D_Gas_Master_Emitter
   simTarget=GPUComputeSim
   rendererClasses [&NiagaraMeshRendererProperties, /Script/Niagara.NiagaraVolumeRendererProperties]
  }
 ]
 userVariables [
  User.DrawBounds {type=NiagaraBool, defaultValue=true}
  User.ResolutionMaxAxis {type=NiagaraInt32, defaultValue=192}
  User.SourceOffset {type=Vector3f, description="", defaultValue {struct=Vector3f, value=<0 0 100>}}
  User.WorldSpaceSize {type=Vector3f, description="", defaultValue {struct=Vector3f, value=<300 300 600>}}
 ]
}
