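# Niagara system template from NiagaraFluids that simulates a smoke gas on a 3D grid.
# A single GPU compute emitter, Grid3D_Gas_Master_Emitter, carries a mesh renderer and a volume renderer.
# User parameters expose the emit position, world-space domain size, max-axis grid resolution and a toggle for drawing bounds.
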NiagaraSystem /NiagaraFluids/Templates/Gas/3D/Systems/Grid3D_Gas_Smoke
dependencies=null
summary {
 systemName=Grid3D_Gas_Smoke
 emitters [{
   emitterName=Grid3D_Gas_Master_Emitter
   simTarget=GPUComputeSim
   rendererClasses [&NiagaraMeshRendererProperties, /Script/Niagara.NiagaraVolumeRendererProperties]
  }]
 userVariables [
  User.DrawBounds {type=NiagaraBool, defaultValue=true}
  User.Emit Position {type=Vector3f, description="", defaultValue {struct=Vector3f, value=<600 0 75>}}
  User.ResolutionMaxAxis {type=NiagaraInt32, defaultValue=256}
  User.WorldSpaceSize {type=Vector3f, description="", defaultValue {struct=Vector3f, value=<1500 750 1500>}}
 ]
}
